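; Child Blueprint for the Galileo Lobster Kayak blade weapon actor, driving the blade's material effects.
; Turning on plays a timeline that lerps SphereMaskRadiusLerp on the main blade and a second masked blade that stays visible underwater.
; Poise changes and block breaks feed UnstableState on the main blade.
; Bullet blocks pass impact location, game time and duration to the blade material for a brief deflect effect.
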(event EventStartIdleFX
  ;[1 1 1] Add Masked shader so that lobster is able to see through water
  (Rendering|Material|SetScalarParameterValueonMaterials (Variables|Default|GetSM_Galileo_Lobster_Blade) "SphereMaskRadiusLerp" 0.0)
  (|CallParentFunction))

(event EventTurnOnLobster
  (bind (_direction _activation _activationsub) (Timeline|Timeline_2 PlayFromStart)
    (:Update
      (Rendering|Material|SetScalarParameterValueonMaterials (Variables|Default|GetSM_Galileo_Lobster_Blade) "SphereMaskRadiusLerp" (Math|Float|Lerp 0.0 1.0 _activation))
      (Rendering|Material|SetScalarParameterValueonMaterials (Variables|Default|GetSM_Galileo_Lobster_Blade_Masked(VisibleUnderwater)) "SphereMaskRadiusLerp" (Math|Float|Lerp 0.0 1.0 _activationsub)))))

(event EventCurrentPoiseChanged
  (params (CurrentPoise :type Float%20%28double-precision%29))
  (|CallParentFunction CurrentPoise)
  (Rendering|Material|SetScalarParameterValueonMaterials (Variables|Default|GetSM_Galileo_Lobster_Blade) "UnstableState" CurrentPoise))

(event EventBlockBreak
  (|CallParentFunction)
  (Rendering|Material|SetScalarParameterValueonMaterials (Variables|Default|GetSM_Galileo_Lobster_Blade) "UnstableState" 1.0))

(event EventDoBulletBlockEffect
  (params (ImpactLocation :type Vector :default "0, 0, 0"))
  (bind _sm_galileo_lobster_blade (Variables|Default|GetSM_Galileo_Lobster_Blade))
  (|CallParentFunction ImpactLocation)
  (Rendering|Material|SetVectorParameterValueonMaterials _sm_galileo_lobster_blade "ProjectileDeflectLocation" ImpactLocation)
  (Rendering|Material|SetScalarParameterValueonMaterials _sm_galileo_lobster_blade "GameTime" (Utilities|Time|GetGameTimeinSeconds))
  (Rendering|Material|SetScalarParameterValueonMaterials _sm_galileo_lobster_blade "ProjectileDeflectEffectDuration" 0.15))

(fn CreateSkeletalAudioComp
  (bind _returnvalue (AddComponent|NODE_AddFortSkeletalAudioComponent-2))
  (return _returnvalue))

(fn ConstructionScript
  (|CallParentFunction))
