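; Parent Blueprint for the Galileo "Lobster" melee blade weapon in Fortnite, handling its visuals, audio and HUD feedback.
; Replicated blade visibility drives turn-on material and idle FX, toggled by holster events when anim trails are used; pawn anim notifies switch the Niagara trail.
; Impacts infer swing direction from delayed tip positions to orient sparks and decals.
; Poise feeds blade instability and a block-break lens warning, while events toggle prompts, reticle status and sound indicators.
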(event Custom|ReticleStatusDisplay
  (params (Hide :type Boolean))
  ;[1 1 1] Hide Reticule Status Display while attacking with the Lobster (prevents Cannot Do That Now messages during combos)
  (Networking|CanExecuteCosmeticEvents
    (:True
      (Utilities|Time|ClearandInvalidateTimerbyHandle (Variables|Default|GetTimer_ShowReticleStatus))
      (HUD|SetHUDElementVisibility (BlueprintContext|GetFortHUDContext) (Variables|Default|GetT_HudElementToHide) true "None")
      (if (not Hide)
        (bind _returnvalue (Utilities|Time|SetTimerbyEvent (delegate Custom|ShowReticleStatusDisplay) 1.2))
        (Variables|Default|SetTimer_ShowReticleStatus _returnvalue)))))

(event EventBeginPlay
  (bind _returnvalue (Game|GetInstigator))
  (bind _ns_idlefx (Variables|Default|GetNS_IdleFX))
  ;[1 1 1] Event Begin Play
  (Utilities|FlowControl|Sequence
    (:then_0
      ;;[0.7339 0 1] Recheck pending fire on blocking tag removal, since this can replicate after the check occurs on the client.
      (Utilities|IsValid _returnvalue
        (:Is%20Valid
          (bind _asynctaskproxy (Ability|Tasks|WaitGameplayTagRemoveFromActor _returnvalue "(TagName=\"Granted.Galileo.Lobster.InWater\")")
            (:then
              (Variables|Default|SetWait%20Leave%20Water%20Async%20Action _asynctaskproxy))
            (:Removed
              (FortWeapon|CheckPendingFire))))))
    (:then_1
      (Components|Activation|Deactivate _ns_idlefx)
      (Utilities|IsValid _ns_idlefx
        (:Is%20Valid
          (Effects|Components|ParticleSystem|SetAutoAttachmentParameters (Variables|Default|GetNS_IdleFX) (Variables|Visual|GetWeaponMesh) (Variables|Effects|GetIdleFXAttachSocket) (Variables|Effects|GetIdleFX%20Location%20Rule) (Variables|Effects|GetIdle%20FX%20Rotation%20Rule) (Variables|Effects|GetIdle%20FX%20Scale%20Rule)))
        (:Is%20Not%20Valid
          (Components|DestroyComponent _ns_idlefx))))
    (:then_2
      (bind _asfort_player_pawn (Utilities|Casting|CastToFortPlayerPawn (Game|GetInstigator)))
      (Variables|Default|SetPlayerCharacter _asfort_player_pawn)
      (if (Pawn|IsLocallyControlled _asfort_player_pawn)
        (bind _component (CallFunction|CreateSkeletalAudioComp))
        (Variables|Default|SetSkeletalAudioComponent _component)))))

(event Custom|ShowReticleStatusDisplay
  ;[1 1 1] Hide Reticule Status Display while attacking with the Lobster (prevents Cannot Do That Now messages during combos)
  (if (not (GameplayTags|HasAnyMatchingGameplayTags (Ability|GetAbilitySystemComponent (Variables|Default|GetPlayerCharacter)) (Variables|Default|GetT_MeleeAttackAbility)))
    (HUD|SetHUDElementVisibility (BlueprintContext|GetFortHUDContext) (Variables|Default|GetT_HudElementToHide) false "None")))

(event Custom|ToggleDodgePrompt
  (params (Visible :type Boolean))
  ;[1 1 1] Toggle Dodge Prompt
  (WeaponHUDData|SetHudKeyActionVisibility 2 Visible))

(event Custom|ToggleAttackPrompt
  (params (Visible :type Boolean))
  ;[1 1 1] Toggle Attack Prompt
  (WeaponHUDData|SetHudKeyActionVisibility 0 Visible))

(event EventOnPlayWeaponFireFX
  (params (bPersistentFire :type Boolean) (bSecondaryFire :type Boolean))
  ;[1 1 1] Event on Play Weapon Fire FX - called from Combat Melee Montage AnimNotify PlayFireFX
  (FortWeapon|PlayForceFeedback false))

(event FortWeapon|EventOnPlayImpactFX
  (params (HitResult :type Hit%20Result%20Structure%20%28by%20ref%29) (ImpactPhysicalSurface :type EPhysicalSurface%20Enum :default "SurfaceType_Default") (SpawnedPSC :type FXSystem%20Component%20Object%20Reference))
  (bind _scene (Variables|Default|GetScene))
  (bind _returnvalue (Transformation|GetWorldLocation _scene))
  (bind (_bblockinghit _binitialoverlap _time _distance _location _impactpoint _normal _impactnormal _physmat _hitactor _hitcomponent _hitbonename _bonename _hititem _elementindex _faceindex _tracestart) (Collision|BreakHitResult HitResult))
  ;[1 1 1] If hitting the world, leave particles along swing direction
  (|CallParentFunction HitResult ImpactPhysicalSurface SpawnedPSC)
  ;;[1 0 0.4181] Looks at where sword tip was when it hits vs right after it hits to find direction
  (Variables|Default|SetImpactTipLocation _returnvalue)
  (Utilities|FlowControl|Delay 0.04)
  (CallFunction|SpawnImpactParticleEffects (Math|Rotator|MakeRotfromX (- _returnvalue (Variables|Default|GetImpactTipLocation))) _impactpoint _hitactor _tracestart _impactnormal (Math|Vector|Normalize (- _returnvalue (Variables|Default|GetImpactTipLocation)))))

(event Custom|ToggleDynamicAudio
  (params (CanPlayIdleAudio :type Boolean))
  ;[1 1 1] dynamic ambient audio toggle - turns off during abilities.
  (bind _asfort_player_pawn (Utilities|Casting|CastToFortPlayerPawn (Game|GetInstigator)))
  (bind _skeletalaudiocomponent (Variables|Default|GetSkeletalAudioComponent)
    (:then
      (Components|SkeletalAudio|SetCanPlay _skeletalaudiocomponent (and (Pawn|IsLocallyControlled _asfort_player_pawn) CanPlayIdleAudio)))
    (:else)))

(event Custom|SetBladeVisibility
  (params (Visible :type Boolean))
  ;[1 1 1] Set Blade Visibility
  (Variables|Default|SetBladeVisibility Visible))

(event Custom|ToggleBlockPrompt
  (params (Visible :type Boolean))
  ;[1 1 1] Toggle Block Prompt
  (WeaponHUDData|SetHudKeyActionVisibility 1 Visible))

(event Custom|StartIdleFX
  ;[1 1 1] Lobster Turn On FX logic
  (Rendering|Material|SetScalarParameterValueonMaterials (Variables|Default|GetSM_Galileo_Lobster_Blade) "SphereMaskRadiusLerp" 0.0)
  (Components|Activation|Deactivate (Variables|Default|GetNS_IdleFX))
  (CallFunction|TurnonLobster)
  (Utilities|FlowControl|Delay 0.35)
  (Components|Activation|Activate (Variables|Default|GetNS_IdleFX) true))

(event Custom|Event%20On%20Block%20Activate
  ;[1 1 1] Block and dodge events for Audio Visualizers
  (SoundIndicator|ShowSoundIndicator (Variables|SoundIndicator|GetSoundIndicatorComponent) (Transformation|GetActorLocation) "/Galileo/Items/Lobster/Assets/Sounds/Default/Lobster_Default_PoiseStart_Cue.Lobster_Default_PoiseStart_Cue" 1.0 0))

(event FortWeapon|EventOnEquippedAndAttached
  ;[1 1 1] Equip / Unequip
  (CallFunction|BindSwingAndAnimTrailEvents)
  (if (Variables|Effects|GetUseAnimTrails)
    (bind _asfort_player_pawn (Utilities|Casting|CastToFortPlayerPawn (Game|GetInstigator)))
    (EventDispatchers|BindEventtoOnHolsteredEvent _asfort_player_pawn (EventDispatchers|CreateEvent :!event Holstered))
    (EventDispatchers|BindEventtoOnUnholsteredEvent _asfort_player_pawn (EventDispatchers|CreateEvent :!event Unholstered))
    (else
      (bind _asfort_player_pawn_2 (Utilities|Casting|CastToFortPlayerPawn (Game|GetInstigator)))
      (Animation|AnimTrailsNotify _asfort_player_pawn_2 false)
      (Animation|AnimTrailsDisable _asfort_player_pawn_2))))

(event Custom|TurnOnLobster
  ;[1 1 1] Lobster Turn On FX logic
  (bind (_direction _activation) (Timeline|Timeline_1 PlayFromStart)
    (:Update
      (Rendering|Material|SetScalarParameterValueonMaterials (Variables|Default|GetSM_Galileo_Lobster_Blade) "SphereMaskRadiusLerp" (Math|Float|Lerp 0.0 1.0 _activation)))))

(event Custom|Event%20On%20Evade
  ;[1 1 1] Block and dodge events for Audio Visualizers
  (SoundIndicator|ShowSoundIndicator (Variables|SoundIndicator|GetSoundIndicatorComponent) (Transformation|GetActorLocation) "/Galileo/Items/Lobster/Assets/Sounds/Default/Lobster_Default_Dodge_Cue.Lobster_Default_Dodge_Cue" 1.0 0))

(event FortWeapon|EventOnUnEquippedAndDetached
  ;[1 1 1] Equip / Unequip
  (CallFunction|UnbindSwingandAnimTrailEvents)
  (bind _asfort_player_pawn (Utilities|Casting|CastToFortPlayerPawn (Game|GetInstigator)))
  (Animation|AnimTrailsNotify _asfort_player_pawn false)
  (Animation|AnimTrailsDisable _asfort_player_pawn))

(event Custom|InstantOnFX
  ;[1 1 1] Lobster Turn On FX logic
  (Rendering|Material|SetScalarParameterValueonMaterials (Variables|Default|GetSM_Galileo_Lobster_Blade) "SphereMaskRadiusLerp" 1.344504)
  (Components|Activation|Activate (Variables|Default|GetNS_IdleFX) false)
  (Class|WeaponEquipSoundComponentLobster|PlayAudioInstantFX (Variables|Default|GetWeaponEquipSoundComponent_Lobster)))

(event Custom|DoBulletBlockEffect
  (params (ImpactLocation :type Vector))
  (bind _playercharacter (Variables|Default|GetPlayerCharacter))
  (bind _currentpoise (Variables|Default|GetCurrentPoise))
  (bind _sm_galileo_lobster_blade (Variables|Default|GetSM_Galileo_Lobster_Blade))
  ;[1 1 1] Bullet Block Effect
  (Rendering|Material|SetVectorParameterValueonMaterials _sm_galileo_lobster_blade "ProjectileDeflectLocation" ImpactLocation)
  (Rendering|Material|SetScalarParameterValueonMaterials _sm_galileo_lobster_blade "GameTime" (Utilities|Time|GetGameTimeinSeconds))
  (Rendering|Material|SetScalarParameterValueonMaterials _sm_galileo_lobster_blade "ProjectileDeflectEffectDuration" 0.15)
  (Development|PrintString (Utilities|String|Append "Current Poise" (Utilities|String|ToString(Float) _currentpoise)) false false "(R=1.000000,G=0.000000,B=0.904101,A=1.000000)")
  (if (and (< _currentpoise (Variables|Default|GetPoiseThresholdToShowWarning)) (!= _currentpoise 0.0))
    (Development|PrintString "DoBlockBreakWarning" false false "(R=0.136643,G=1.000000,B=0.000000,A=1.000000)" 15.0)
    (Utilities|IsValid _playercharacter
      (:Is%20Valid
        (bind _returnvalue (Feedback|SpawnGenericCameraLensEffect _playercharacter "/Galileo/Items/Lobster/Assets/Camera_Effects/BP_Camera_Lens_Galileo_BlockBreak.BP_Camera_Lens_Galileo_BlockBreak_C"))
        (bind _asniagara_lens_effect_base (Utilities|Casting|CastToNiagaraLensEffectBase _returnvalue)
          (:then
            (Variables|Default|SetAs%20Niagara%20Lens%20Effect%20Base _asniagara_lens_effect_base))
          (:CastFailed
            (Development|PrintString "Expected Camera Lens Effect to be an NiagaraLensEffectBase" true true "(R=1.000000,G=0.117693,B=0.000000,A=1.000000)")))
        (bind _returnvalue_2 (Utilities|Time|SetTimerbyEvent (EventDispatchers|CreateEvent :!event Custom|DeactivateBlockBreakWarning) 3.0))
        (Variables|Default|SetTimer_BlockBreakWarning _returnvalue_2)))))

(event FortWeapon|EventOnLocalUnEquip
  ;[1 1 1] Equip / Unequip
  (CallFunction|UnbindSwingandAnimTrailEvents)
  (bind _asfort_player_pawn (Utilities|Casting|CastToFortPlayerPawn (Game|GetInstigator)))
  (EventDispatchers|UnbindEventfromOnHolsteredEvent _asfort_player_pawn (EventDispatchers|CreateEvent :!event Holstered))
  (EventDispatchers|UnbindEventfromOnUnholsteredEvent _asfort_player_pawn (EventDispatchers|CreateEvent :!event Unholstered))
  (if (Variables|Default|GetBladeVisibility)
    (CallFunction|SetBladeVisibility false)))

(event Custom|DeactivateBlockBreakWarning
  ;[1 1 1] Deactivate Block Break
  (Utilities|Time|ClearandInvalidateTimerbyHandle (Variables|Default|GetTimer_BlockBreakWarning))
  (bind _as_niagara_lens_effect_base (Variables|Default|GetAs%20Niagara%20Lens%20Effect%20Base)
    (:then
      (Utilities|IsValid (Variables|/Script/Niagara.NiagaraActor|GetNiagaraComponent _as_niagara_lens_effect_base)
        (:Is%20Valid
          (Components|Activation|Deactivate (Variables|/Script/Niagara.NiagaraActor|GetNiagaraComponent _as_niagara_lens_effect_base)))))
    (:else)))

(event Custom|CurrentPoiseChanged
  (params (CurrentPoise :type Float%20%28double-precision%29))
  ;[1 1 1] Poise Changed
  (bind _output_get (Variables|Default|SetCurrentPoise CurrentPoise))
  (Rendering|Material|SetScalarParameterValueonMaterials (Variables|Default|GetSM_Galileo_Lobster_Blade) "UnstableState" CurrentPoise)
  (Niagara|SetNiagaraVariable(Float) (Variables|Default|GetNS_IdleFX) "UnstableState" _output_get))

(event EventOnWeaponVisibilityChanged
  (params (bVisible :type Boolean) (bSetForLocalControllerOnly :type Boolean))
  (|CallParentFunction bVisible bSetForLocalControllerOnly)
  (if (!= bVisible (Variables|Default|GetBladeVisibility))
    (Variables|Default|SetBladeVisibility bVisible)))

(event Custom|AnimTrailNotifyBegin
  (params (bActive :type Boolean))
  ;[1 1 1] Niagara Anim Trail Activation
  (Niagara|SetNiagaraVariable(Bool) (Variables|Default|GetNS_AnimTrail) "AnimTrailActivate" bActive))

(event Custom|BlockBreak
  ;[1 1 1] Poise Changed
  (Rendering|Material|SetScalarParameterValueonMaterials (Variables|Default|GetSM_Galileo_Lobster_Blade) "UnstableState" 1.0)
  (Niagara|SetNiagaraVariable(Float) (Variables|Default|GetNS_IdleFX) "UnstableState" 1.0)
  (CallFunction|DeactivateBlockBreakWarning))

(event Custom|AnimTrailNotifyEnd
  ;[1 1 1] Niagara Anim Trail Activation
  (Niagara|SetNiagaraVariable(Bool) (Variables|Default|GetNS_AnimTrail) "AnimTrailActivate" false))

(event EventEndPlay
  (params (EndPlayReason :type EEndPlayReason%20Enum :default "Destroyed"))
  ;[1 0.4252 0] EndPlay: Cancel Async Actions
  (AsyncAction|Cancel (Variables|Default|GetWait%20Leave%20Water%20Async%20Action)))

(fn BindSwingandAnimTrailEvents
  (bind _returnvalue (Game|GetInstigator))
  (Networking|CanExecuteCosmeticEvents
    (:True
      (bind _asplayer_pawn_athena (Utilities|Casting|CastToPlayerPawn_Athena _returnvalue)
        (:then
          ;[1 1 1] Bind Swing Trail Events for PlayerPawn _Athena
          (Utilities|IsValid (Niagara|GetNiagaraSystemAsset (Variables|Default|GetNS_AnimTrail))
            (:Is%20Valid
              (Default|BindEventtoAnimNotifyBegin _asplayer_pawn_athena (EventDispatchers|CreateEvent :!event AnimTrailNotifyBegin))
              (Default|BindEventtoAnimNotifyEnd _asplayer_pawn_athena (EventDispatchers|CreateEvent :!event AnimTrailNotifyEnd)))) ; If we actually have a Niagara system to play...
        )
        (:CastFailed
          (bind _asplayer_pawn_generic (Utilities|Casting|CastToPlayerPawn_Generic _returnvalue))
          ;[1 1 1] Bind Swing Trail Events for PlayerPawn_Generic
          (Utilities|IsValid (Niagara|GetNiagaraSystemAsset (Variables|Default|GetNS_AnimTrail))
            (:Is%20Valid
              (Default|BindEventtoAnimNotifyBegin _asplayer_pawn_generic (EventDispatchers|CreateEvent :!event AnimTrailNotifyBegin))
              (Default|BindEventtoAnimNotifyEnd _asplayer_pawn_generic (EventDispatchers|CreateEvent :!event AnimTrailNotifyEnd)))) ; If we actually have a Niagara system to play...
        )))) ; No need to register on the server, as we\nonly need these events for cosmetic purposes.
)

(fn CreateSkeletalAudioComp
  (bind _returnvalue (AddComponent|NODE_AddFortSkeletalAudioComponent-3))
  (return _returnvalue))

(fn GetAffiliationsToShowFor
  (return (Utilities|Array|MakeArray "Neutral" "Hostile")))

(fn Holstered
  (bind _asfort_player_pawn (Utilities|Casting|CastToFortPlayerPawn (Game|GetInstigator)))
  (EventDispatchers|UnbindEventfromOnHolsteredEvent _asfort_player_pawn (EventDispatchers|CreateEvent :!event Holstered))
  (EventDispatchers|BindEventtoOnUnholsteredEvent _asfort_player_pawn (EventDispatchers|CreateEvent :!event Unholstered))
  (CallFunction|SetBladeVisibility false))

(fn OnRep_BladeVisibility
  (if (Variables|Default|GetBladeVisibility)
    (if (Variables|Default|GetInstantOn)
      (CallFunction|InstantonFX)
      (else
        (CallFunction|StartIdleFX)))
    (else
      (Components|Activation|Deactivate (Variables|Default|GetNS_IdleFX)))))

(fn OnRep_InstantOn)

(fn ShouldShowSoundIndicator
  (params (PlayerController :type Fort%20Player%20Controller%20Object%20Reference))
  (return true))

(fn SpawnImpactParticleEffects
  (params (Rotation%20from%20Tip :type Rotator) (ImpactPoint :type Vector) (Hit%20Actor :type Actor%20Object%20Reference) (Trace%20Start :type Vector) (Impact%20Normal :type Vector) (Swing%20Direction :type Vector))
  (Utilities|Casting|CastToFortPlayerPawn Hit%20Actor
    (:then
      ;[1 1 1] Hit Player Pawn
      (bind _returnvalue (Niagara|SpawnSystematLocation "/Galileo/Items/Lobster/Assets/FX/NS_Galileo_Lobster_Impact.NS_Galileo_Lobster_Impact" (+ (Variables|GetImpactPoint) "0, 0,35.000000") (Variables|GetRotation%20from%20Tip)))
      (Niagara|SetNiagaraVariable(LinearColor) _returnvalue "Color_Blade_A" (Variables|Default|GetBlade%20Color%20Hot))
      (Niagara|SetNiagaraVariable(LinearColor) _returnvalue "Color_Blade_B" (Variables|Default|GetBlade%20Color%20Cool))
      (Niagara|SetNiagaraVariable(Bool) _returnvalue "HitPlayer" true))
    (:CastFailed
      ;[1 1 1] Hit world objects and left scorched decal
      (Utilities|Casting|CastToB_Galileo_Lobster_Block_Collision (Variables|GetHit%20Actor)
        (:then
          (Development|PrintString "Hit Blocking Player and Skip impact FX" false false))
        (:CastFailed
          (Utilities|FlowControl|Sequence
            (:then_0
              (bind _returnvalue_2 (Niagara|SpawnSystematLocation "/Galileo/Items/Lobster/Assets/FX/NS_Galileo_Lobster_Impact.NS_Galileo_Lobster_Impact" (Variables|GetImpactPoint) (Variables|GetRotation%20from%20Tip)))
              (Niagara|SetNiagaraVariable(LinearColor) _returnvalue_2 "Color_Blade_A" (Variables|Default|GetBlade%20Color%20Hot))
              (Niagara|SetNiagaraVariable(LinearColor) _returnvalue_2 "Color_Blade_B" (Variables|Default|GetBlade%20Color%20Cool))
              (Niagara|SetNiagaraVariable(Bool) _returnvalue_2 "HitPlayer" false))
            (:then_1
              ;;[1 1 0] Don't spawn decal gameplay tag check
              (bind _returnvalue_3 (GameplayTags|HasAnyMatchingGameplayTags(Message) (Variables|GetHit%20Actor) (Variables|Default|GetTagsToNotSpawnDecal)))
              (if (not _returnvalue_3)
                (bind _returnvalue_4 (Rendering|Decal|SpawnDecalatLocation (Variables|Default|GetImpactDecal) "32.000000,64.000000,6.000000" (Variables|GetImpactPoint) (Math|Rotator|MakeRotfromXY (Variables|GetImpact%20Normal) (Variables|GetSwing%20Direction)) 2.0))
                (Rendering|Components|Decal|SetFadeOut _returnvalue_4 0.0 2.0 false)))))))))

(fn UnbindSwingAndAnimTrailEvents
  (bind _returnvalue (Game|GetInstigator))
  (bind _asplayer_pawn_athena (Utilities|Casting|CastToPlayerPawn_Athena _returnvalue)
    (:then
      (Default|UnbindEventfromAnimNotifyBegin _asplayer_pawn_athena (EventDispatchers|CreateEvent :!event AnimTrailNotifyBegin))
      (Default|UnbindEventfromAnimNotifyEnd _asplayer_pawn_athena (EventDispatchers|CreateEvent :!event AnimTrailNotifyEnd)))
    (:CastFailed
      (bind _asplayer_pawn_generic (Utilities|Casting|CastToPlayerPawn_Generic _returnvalue)
        (:then
          (Default|UnbindEventfromAnimNotifyBegin _asplayer_pawn_generic (EventDispatchers|CreateEvent :!event AnimTrailNotifyBegin))
          (Default|UnbindEventfromAnimNotifyEnd _asplayer_pawn_generic (EventDispatchers|CreateEvent :!event AnimTrailNotifyEnd)))
        (:CastFailed
          (Development|PrintString "Unbind REALLY Failed" true true "(R=1.000000,G=1.000000,B=1.000000,A=1.000000)" 4.0))))))

(fn Unholstered
  (bind _asfort_player_pawn (Utilities|Casting|CastToFortPlayerPawn (Game|GetInstigator)))
  (EventDispatchers|UnbindEventfromOnUnholsteredEvent _asfort_player_pawn (EventDispatchers|CreateEvent :!event Unholstered))
  (EventDispatchers|BindEventtoOnHolsteredEvent _asfort_player_pawn (EventDispatchers|CreateEvent :!event Holstered))
  (CallFunction|SetBladeVisibility true))

(fn ConstructionScript
  (bind _sm_galileo_lobster_blade (Variables|Default|GetSM_Galileo_Lobster_Blade))
  (|CallParentFunction)
  (Rendering|Material|SetScalarParameterValueonMaterials _sm_galileo_lobster_blade "HueColorShift" (Variables|Default|GetBladeHueColorShift))
  (Rendering|Material|SetScalarParameterValueonMaterials _sm_galileo_lobster_blade "RedBlueSwitch" (Variables|Default|GetBladeRedBlueSwitch))
  (Rendering|Material|SetScalarParameterValueonMaterials _sm_galileo_lobster_blade "EmissiveIntensity" (Variables|Default|GetBladeEmissiveIntensity)))
