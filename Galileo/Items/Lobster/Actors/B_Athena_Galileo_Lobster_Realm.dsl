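; Lightsaber weapon actor for Galileo "Lobster", supporting a throw with a homing target, a Throw/Attack HUD prompt and a custom reticle.
; The replicated thrown projectile acts as the "realm" holster source on equip and unequip and gates reticle updates.
; A looping timer refreshes the reticle's in-range state, and a client timeline flashes it on projectile hits.
; The server cancels tactical sprint for blocking, and construction sets blade hue and glow.
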(event Custom|ServerCancelTacSprint
  ;[1 0.6954 0.3789] Cancel Tactical Sprint Ability
  ; For use by GA_Realm_CheckIfTryingToBlock and allows cancelling tac sprint ability if trying to block while tac sprinting. (The block GA itself cannot cancel the tac sprint as it does not activate since the weapon is holstered while tac sprinting.)
  (Networking|SwitchHasAuthority
    (:Authority
      (bind _playercharacter (Variables|Default|GetPlayerCharacter)
        (:then
          (Utilities|IsValid (Ability|GetFortAbilitySystemComponent _playercharacter)
            (:Is%20Valid
              (Tags|CancelAbilitiesWithTags (Ability|GetFortAbilitySystemComponent _playercharacter) (GameplayTags|MakeLiteralGameplayTagContainer "(GameplayTags=((TagName=\"Gameplay.Action.Player.TacticalSprint\")))")))))
        (:else)))))

(event Custom|ReticleTick
  ;[1 1 1] Reticle State Updates
  (CallFunction|ReticleUpdate))

(event FortWeapon|EventOnEquippedAndAttached
  ;[0.1284 1 0] Event OnEquippedAndAttached, initialize the saber throw prompt.
  (|CallParentFunction)
  (bind _playercharacter (Variables|Default|GetPlayerCharacter)
    (:then
      (bind _projectile_replicated (Variables|Default|GetProjectile%20Replicated)
        (:then
          (Weapon|SetHolsterWeapon _playercharacter _projectile_replicated true false "realm") ; If the projectile is out,  holster the weapon
        )
        (:else
          _))
      (if (Pawn|IsLocallyControlled _playercharacter)
        (Networking|CanExecuteCosmeticEvents
          (:True
            (CallFunction|ToggleSaberThrowPrompt false)))))
    (:else)))

(event FortWeapon|EventOnUnEquippedAndDetached
  ;[1 0.5973 0.2561] Event OnUnEquippedAndDetached - Remove the "realm" holster id if the lightsaber is unequipped. (The lightsaber is holstered with this id at the beginning of the throw.) This prevents the player from having unusuable and invisible weapons if they switch weapons while the lightsaber is midthrow
  (|CallParentFunction)
  (Utilities|FlowControl|Sequence
    (:then_0
      (Variables|Default|SetInstantOn false)
      (bind _projectile_replicated (Variables|Default|GetProjectile%20Replicated)
        (:then
          (bind _asfort_player_pawn (Utilities|Casting|CastToFortPlayerPawn (Game|GetInstigator)))
          (Weapon|SetHolsterWeapon _asfort_player_pawn _projectile_replicated false false "realm") ; Projectile is the holster source (if the projectile doesn't exist, the weapon won't be holstered)
        )
        (:else)))
    (:then_1
      (CallFunction|UpdateReticleVisibility)
      (Utilities|Time|ClearandInvalidateTimerbyHandle (Variables|Default|GetReticleTickTimer)))))

(event Custom|ToggleSaberThrowPrompt
  (params (Visible :type Boolean))
  ;[0.4828 0 1] Toggle Saber Throw Prompt
  (bind _output_get (Variables|Default|SetIsInThrowableState Visible))
  (WeaponHUDData|SetHudKeyActionDescription 0 (select _output_get "NSLOCTEXT(\"[98880BF64D305FF12F7A14B586FD99EF]\", \"ACE34533478EFD4C2FD1AAB9A83B3131\", \"Throw\")" "NSLOCTEXT(\"[98880BF64D305FF12F7A14B586FD99EF]\", \"096A1E284E1D300F7AA2459D79A74305\", \"Attack\")")) ; changes the text in the Control Propmt HUD
  (CallFunction|UpdateReticleVisibility))

(event Custom|HandleProjectileHit
  ;[0.1669 0 1] HandleProjectileHit(): calls Client_HandleProjectileHit()
  ; Client_HandleProjectileHit(): plays a timeline that shows the ReticleHitColor going from 1 to 0.
  (Networking|SwitchHasAuthority
    (:Authority
      (CallFunction|ClientHandleProjectileHit))))

(event Custom|Client%20HandleProjectileHit
  ;[0.1669 0 1] HandleProjectileHit(): calls Client_HandleProjectileHit()
  ; Client_HandleProjectileHit(): plays a timeline that shows the ReticleHitColor going from 1 to 0.
  (Networking|CanExecuteCosmeticEvents
    (:True
      (bind (_direction _hitpercent) (Timeline|Projectile%20Hit%20Hostile%20Timeline PlayFromStart)
        (:Update
          (Variables|Default|SetReticleHitPercent _hitpercent))
        (:Finished
          (Variables|Default|SetReticleHitPercent 0.0))))))

(event Custom|SetProjectile
  (params (Projectile :type B%20Prj%20Realm%20Thrown%20Lobster%20Object%20Reference))
  ;[0 0.1718 1] SetProjectile
  (Variables|Default|SetProjectile%20Replicated Projectile)
  (Variables|Default|SetHomingTarget 0) ; Clear homing target after proj is spawned
)

(fn GetAffiliationsToShowFor
  (return "Friendly"))

(fn GetReticleColorOverride
  (params (HitResults :type Array%20of%20Hit%20Result%20Structures)))

(fn OnRep_ProjectileReplicated
  (Networking|CanExecuteCosmeticEvents
    (:True
      (CallFunction|UpdateReticleVisibility))))

(fn ReticleUpdate
  (Variables|Default|GetProjectile%20Replicated
    (:then
      _)
    (:else
      (bind _homingtarget (Variables|Default|GetHomingTarget)
        (:then
          (LocalVariables|SetTargetInRange (Utilities|IsTargetLikelytobeHit _homingtarget)))
        (:else
          (bind _asfort_player_pawn (Utilities|Casting|CastToFortPlayerPawn (Game|GetInstigator)))
          (LocalVariables|SetTargetInRange (<= (Math|Vector|Distance(Vector) (Transformation|GetActorLocation (Game|GetInstigator)) (Reticle|GetLocationUnderReticle (Fortnite|GetFortPlayerControllerfromActor _asfort_player_pawn))) (ScalableFloat|GetValueatLevel (Variables|Default|GetRow_HomingTargetingMaxRange)))))) ; If we have a homing target,\ncheck if they're likely to be hit.\nTurn reticle red if so.
      (Default|CallUpdateReticleState (LocalVariables|GetTargetInRange) (Utilities|IsValid (Variables|Default|GetHomingTarget) :!class /Script/Engine.KismetSystemLibrary)))) ; Don't change if weapon is thrown
)

(fn ShouldShowSoundIndicator
  (params (PlayerController :type Fort%20Player%20Controller%20Object%20Reference))
  (return false))

(fn UpdateReticleVisibility
  (bind _projectile_replicated (Variables|Default|GetProjectile%20Replicated))
  (bind _returnvalue (Utilities|IsValid _projectile_replicated :!class /Script/Engine.KismetSystemLibrary))
  (bind _isinthrowablestate (Variables|Default|GetIsInThrowableState))
  (bind _returnvalue_2 (CallFunction|IsBeingEquipped))
  (bind _returnvalue_3 (FortWeapon|IsEquipped))
  (bind _returnvalue_4 (Math|Boolean|NORBoolean _returnvalue_3 _returnvalue_2))
  (bind _returnvalue_5 (Math|Boolean|ORBoolean _returnvalue_4 _isinthrowablestate _returnvalue))
  (bind _returnvalue_6 (not _returnvalue_5))
  (Networking|CanExecuteCosmeticEvents
    (:True
      (bind _playercharacter (Variables|Default|GetPlayerCharacter)
        (:then
          (if (Pawn|IsLocallyControlled _playercharacter)
            (HUD|SetHUDElementVisibility (BlueprintContext|GetFortHUDContext) (Variables|Default|GetReticle%20HUD%20Element%20Tag) _returnvalue_6 "None") ; Show the reticle when in the throw state and when the projectile is currently flying
            (if _returnvalue_6
              (Utilities|Time|ClearandInvalidateTimerbyHandle (Variables|Default|GetReticleTickTimer))
              (else
                (bind _returnvalue_7 (Utilities|Time|SetTimerbyEvent (EventDispatchers|CreateEvent :!event ReticleTick) 0.01 true))
                (Variables|Default|SetReticleTickTimer _returnvalue_7)))))
        (:else)))))

(fn ConstructionScript
  (bind _sm_galileo_lobster_blade_masked_visibleunderwater (Variables|Default|GetSM_Galileo_Lobster_Blade_Masked(VisibleUnderwater)))
  (|CallParentFunction)
  (Rendering|Material|SetScalarParameterValueonMaterials _sm_galileo_lobster_blade_masked_visibleunderwater "HueColorShift" (Variables|Default|GetBladeHueColorShift))
  (Rendering|Material|SetScalarParameterValueonMaterials _sm_galileo_lobster_blade_masked_visibleunderwater "EmissiveIntensity" (Variables|Default|GetBladeEmissiveIntensity)))
