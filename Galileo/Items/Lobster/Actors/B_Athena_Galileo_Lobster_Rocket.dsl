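; Blueprint actor for the Galileo Lobster Rocket item in Athena.
; A helper function, CreateSkeletalAudioComp, adds a FortSkeletalAudioComponent to the actor and returns it.
; The construction script only calls its parent, so any other behavior comes from the parent class.
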(fn CreateSkeletalAudioComp
  (bind _returnvalue (AddComponent|NODE_AddFortSkeletalAudioComponent-0))
  (return _returnvalue))

(fn ConstructionScript
  (|CallParentFunction))
